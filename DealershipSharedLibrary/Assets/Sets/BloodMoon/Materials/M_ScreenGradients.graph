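# M_ScreenGradients is an unlit, additive, two-sided effects material for sprite, ribbon and mesh particles, built around a procedural gradient.
# A sphere mask on one centred UV axis, horizontal or vertical, forms an invertible, power-shaped gradient that static switches can swap for textures in color and opacity.
# Emissive takes particle or parameter color, optionally adds base color, then runs through MF_FX_EmissiveAdjustments.
# Opacity combines particle alpha, backface dimming and switchable depth fade with inverted Fresnel.
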Material /DealershipSharedLibrary/Assets/Sets/BloodMoon/Materials/M_ScreenGradients
graph {
 properties {
  bEnableResponsiveAA=true
  bUsedWithNiagaraMeshParticles=true
  bUsedWithNiagaraRibbons=true
  bUsedWithNiagaraSprites=true
  bUsedWithParticleSprites=true
  blendMode=BLEND_Additive
  opacityMaskClipValue=0.03
  pixelDepthOffsetMode=PDOM_Legacy
  shadingModel=MSM_Unlit
  twoSided=true
  usedShadingModels=Unlit
 }
 nodes [
  n1 = ScalarParameter() {defaultValue=1, group=Gradient, parameterName=Gradient - Inverse Gradient}
  n2 = NamedRerouteUsage()
  n3 = ComponentMask(None=@n2) {r=true}
  n4 = SphereMask(
   A=If(
    A=ScalarParameter() {defaultValue=1, group=Gradient, parameterName=Gradient - Horizontal Gradient DIrection, sortPriority=4}
    "A > B"=@n3
    "A == B"=@n3
    "A < B"=ComponentMask(None=@n2) {g=true}
   ) {constB=1}
   B=Constant()
   Radius=ScalarParameter() {defaultValue=0.5, group=Gradient, parameterName=Gradient - Thickness, sortPriority=2}
   Hardness=ScalarParameter() {defaultValue=0.5, group=Gradient, parameterName=Gradient - Hardness, sortPriority=2}
  )
  n5 = OneMinus(None=@n4)
  NamedRerouteDeclaration(None=Power(
    Base=If(A=@n1, "A > B"=@n5, "A == B"=@n5, "A < B"=@n4) {constB=1}
    Exp=ScalarParameter() {defaultValue=1, group=Gradient, parameterName=Gradient - Power}
   )) {name=Gradient, nodeColor=<rgba 1 0 0 1>}
  n6 = ScalarParameter() {group=Base Color, parameterName=Base Color - Use Particle Color?, sortPriority=1}
  n7 = VectorParameter() {group=Base Color, parameterName=Base Color, sortPriority=1, defaultValue=<rgba 1 0 0 0>}
  n8 = ParticleColor()
  n9 = NamedRerouteUsage()
  n10 = TextureCoordinate()
  n11 = VectorParameter() {group=Base Color, parameterName=Base Color - Texture Tiling, channelNames {a=Offset Y, b=Offset X, g=Tile Y, r=Tile X}}
  n12 = AppendVector(A=@n11.Tile X, B=@n11.Tile X)
  n13 = TextureSampleParameter2D(UVs=Add(A=Multiply(A=@n10, B=@n12), B=AppendVector(A=@n11.Offset X, B=@n11.Offset X))) {
   group=Base Color
   parameterName=Base Color - Texture
   sortPriority=2
   texture=/DealershipSharedLibrary/FX/Textures/Noise/T_VCVFX_EnergyCaustic_A
  }
  NamedRerouteDeclaration(None=Multiply(
    A=If(A=@n6, "A > B"=@n7.RGB, "A == B"=@n7.RGB, "A < B"=@n8.RGB) {constB=1}
    B=StaticSwitchParameter(True=@n9, False=@n13.RGB) {defaultValue=true, group=Base Color, parameterName=Base Color - Use Procedural Gradient?, sortPriority=0}
   )) {name=Base Color, nodeColor=<rgba 0.0781 1 0 1>}
  n14 = NamedRerouteUsage()
  n15 = TextureCoordinate()
  n16 = VectorParameter() {group=Opacity, parameterName=Opacity - Texture Tiling, channelNames {a=Offset Y, b=Offset X, g=Tile Y, r=Tile X}}
  n17 = AppendVector(A=@n16.Tile X, B=@n16.Tile X)
  n18 = TextureSampleParameter2D(UVs=Add(A=Multiply(A=@n15, B=@n17), B=AppendVector(A=@n16.Offset X, B=@n16.Offset X))) {group=Opacity, parameterName=Opacity - Texture, sortPriority=2, texture=&T_VCVFX_EnergyCaustic_A}
  n19 = Multiply(
   A=Multiply(
    A=ParticleColor().A
    B=StaticSwitchParameter(True=@n14, False=@n18.RGB) {defaultValue=true, group=Opacity, parameterName=Opacity - Use Procedural Gradient?, sortPriority=1}
   )
   B=ScalarParameter() {defaultValue=1, group=Opacity, parameterName=Opacity - Master Control, sortPriority=0}
  )
  n20 = LinearInterpolate(A=Multiply(A=@n19, B=ScalarParameter() {parameterName=Opacity - BackFace Opacity}), B=@n19, Alpha=TwoSidedSign())
  n21 = DepthFade(Opacity=@n20, FadeDistance=ScalarParameter() {parameterName=Opacity - Depth Fade})
  n22 = ScalarParameter() {defaultValue=5, parameterName=Opacity - Fresnel Exponent}
  n23 = Saturate(None=OneMinus(None=Fresnel(ExponentIn=@n22)))
  NamedRerouteDeclaration(None=StaticSwitchParameter(True=Multiply(A=@n21, B=@n23), False=@n20) {parameterName=Mesh Depth Fade / Fresnel}) {name=Opacity, nodeColor=<rgba 0.6406 1 0 1>}
  n24 = ScalarParameter() {group=Emissive Color, parameterName=Emissive Color - Add Base Color to Emissive Color?, sortPriority=1}
  n25 = ParticleColor()
  n26 = If(
   A=ScalarParameter() {defaultValue=1, group=Emissive Color, parameterName=Emissive Color - Use Particle Color?, sortPriority=2}
   "A > B"=@n25.RGB
   "A == B"=@n25.RGB
   "A < B"=VectorParameter() {group=Emissive Color, parameterName=Emissive Color}.RGB
  ) {constB=1}
  n27 = Add(A=@n26, B=NamedRerouteUsage())
  NamedRerouteDeclaration(None=MaterialFunctionCall(Emissive=If(A=@n24, "A > B"=@n27, "A == B"=@n27, "A < B"=@n26) {constB=1}) {materialFunction=&MF_FX_EmissiveAdjustments}.Result) {name=Emissive Color, nodeColor=<rgba 0 0.8828 1 1>}
  n28 = TextureCoordinate()
  n29 = VectorParameter() {group=Gradient, parameterName=Gradient - Offset}
  NamedRerouteDeclaration(None=Add(A=Add(A=@n28) {constB=-0.5}, B=AppendVector(A=@n29.R, B=@n29.R))) {name=Gradient UVs, nodeColor=<rgba 0.3359 1 0 1>}
  NamedRerouteUsage()
  NamedRerouteUsage()
  NamedRerouteUsage()
 ]
}
